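# Translucent, two-sided unlit material for the Genie wheel cosmetic's swirling trail effect, usable on Niagara sprites, mesh particles and skeletal meshes.
# Rotating trail and cloud textures, distorted by a motion-vector-style texture, build gradient-tinted emissive and masked opacity, alongside panning noise.
# A rotating, sphere-masked pulse builds world position offset.
# Dynamic parameters and particle color feed named reroutes, and the attributes pass through vehicle damage and de-rez functions.
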Material /VehicleCosmetics/Wheels/Genie/Materials/M_Wheel_Genie_FX
graph {
 properties {
  bAutomaticallySetUsageInEditor=false
  bUseMaterialAttributes=true
  bUsedWithNiagaraMeshParticles=true
  bUsedWithNiagaraSprites=true
  bUsedWithSkeletalMesh=true
  blendMode=BLEND_Translucent
  previewMesh=/VehicleCosmetics/Wheels/Genie/Meshes/SK_Wheel_Genie
  shadingModel=MSM_Unlit
  twoSided=true
  usedShadingModels=Unlit
 }
 nodes [
  n1 = NamedRerouteUsage()
  n2 = NamedRerouteUsage()
  n3 = NamedRerouteUsage()
  MaterialFunctionCall(In=MaterialFunctionCall(Material=SetMaterialAttributes!(MaterialAttributes, World Position Offset=@n1, Emissive Color=@n2, Opacity=@n3) {attributeSetTypes [#1, #2, #3]}) {materialFunction=&MF_BaseVehicle_Damage}.Result) {materialFunction=&MF_VehicleDeRez_Transition}
  n4 = DynamicParameter() {paramNames [Desaturation,Power,Offset,TileSpeed]}
  NamedRerouteDeclaration(None=@n4.TileSpeed) {name=TileSpeed, nodeColor=<rgba 0.0781 0 1 1>}
  NamedRerouteDeclaration(None=@n4.Desaturation) {name=Desaturation, nodeColor=<rgba 0.0781 0 1 1>}
  n5 = Constant()
  n6 = ComponentMask(None=Rotator(Coordinate=TextureCoordinate()) {speed=-1}) {r=true}
  n7 = Multiply(A=@n6, B=OneMinus(None=@n6))
  n8 = Constant() {r=6}
  n9 = Multiply(A=Clamp(None=Power(Base=@n7, Exp=@n8)), B=Constant() {r=1500})
  n10 = SphereMask(A=TextureCoordinate(), B=Constant2Vector() {g=0.5, r=0.5}, Radius=Constant() {r=0.5}) {hardnessPercent=0}
  n11 = Multiply(A=Power(Base=@n10, Exp=Constant() {r=6}), B=Constant() {r=5})
  n12 = Multiply(A=Multiply(A=Clamp(None=@n9), B=Subtract(A=@n10, B=@n11)), B=Constant() {r=20})
  NamedRerouteDeclaration(None=AppendVector(A=AppendVector(A=@n5, B=@n12), B=@n5)) {name=WPO, nodeColor=<rgba 0.2656 0 1 1>}
  NamedRerouteDeclaration(None=@n4.Power) {name=Power, nodeColor=<rgba 0.0781 0 1 1>}
  NamedRerouteDeclaration(None=@n4.Offset) {name=Offset, nodeColor=<rgba 0.0781 0 1 1>}
  n13 = NamedRerouteUsage()
  n14 = Constant()
  n15 = Multiply(
   A=Add(A=AppendVector(A=@n13, B=@n14), B=TextureCoordinate() {uTiling=10, vTiling=10})
   B=AppendVector(A=Constant() {r=1}, B=Constant() {r=0.1})
  )
  n16 = TextureCoordinate() {uTiling=10, vTiling=10}
  n17 = Multiply(
   A=TextureSample(UVs=Panner(Coordinate=@n15) {speedX=-1.2}) {samplerSource=SSM_Wrap_WorldGroupSettings, texture=/VehicleCosmetics/Wheels/_SharedTextures/FX/Noise_Generic_01_Pack}.B
   B=TextureSample(UVs=Panner(Coordinate=@n16) {speedX=-0.1}) {samplerSource=SSM_Wrap_WorldGroupSettings, texture=&Noise_Generic_01_Pack}.B
  )
  NamedRerouteDeclaration(None=Multiply(A=Clamp(None=@n17), B=Constant() {r=4000})) {name=Noise, nodeColor=<rgba 0.0078 1 0 1>}
  n18 = Constant3Vector() {constant=<rgba 0.01 0.2 1 0>}
  n19 = ComponentMask(None=TextureCoordinate()) {r=true}
  n20 = Power(Base=OneMinus(None=@n19), Exp=NamedRerouteUsage())
  n21 = Constant3Vector() {constant=<rgba 1 0.15 1 0>}
  n22 = Desaturation(None=Add(A=Multiply(A=@n18, B=@n19), B=Multiply(A=@n20, B=@n21)), Fraction=NamedRerouteUsage())
  n23 = Multiply(A=@n22, B=Constant() {r=0.2})
  n24 = TextureCoordinate()
  n25 = ComponentMask(None=TextureSample(UVs=Rotator(Coordinate=@n24) {speed=0.35}) {samplerSource=SSM_Wrap_WorldGroupSettings, texture=/VehicleCosmetics/Wheels/_SharedTextures/FX/CircleMV_RGB_T}.RGB) {g=true, r=true}
  n26 = Constant() {r=0.2}
  n27 = Multiply(A=@n25, B=@n26)
  n28 = Rotator(Coordinate=Add(A=TextureCoordinate(), B=@n27), Time=Multiply(A=NamedRerouteUsage(), B=Time())) {speed=2}
  n29 = Multiply(
   A=TextureSample(UVs=@n28) {samplerSource=SSM_Wrap_WorldGroupSettings, texture=/VehicleCosmetics/Wheels/_SharedTextures/FX/Trail3_RBG_T}.RGB
   B=Constant() {r=30}
  )
  NamedRerouteDeclaration(None=Clamp(None=Multiply(A=@n23, B=@n29))) {name=Warp Outline, nodeColor=<rgba 0 1 0.4141 1>}
  n30 = ParticleColor()
  NamedRerouteDeclaration(None=@n30.R) {name=VC_R, nodeColor=<rgba 1 0 0.9609 1>}
  NamedRerouteDeclaration(None=@n30.A) {name=VC_A, nodeColor=<rgba 1 0 0.9609 1>}
  n31 = ComponentMask(None=TextureCoordinate()) {g=true}
  n32 = Multiply(A=VectorParameter() {parameterName=Color_B, defaultValue=<rgba 0 0.8307 1 1>}.RGB, B=@n31)
  n33 = Multiply(
   A=Power(Base=OneMinus(None=@n31), Exp=NamedRerouteUsage())
   B=VectorParameter() {parameterName=Color_A, defaultValue=<rgba 0.998 0.1104 1 1>}.RGB
  )
  n34 = NamedRerouteUsage()
  n35 = ComponentMask(None=TextureCoordinate()) {r=true}
  n36 = Constant() {r=2}
  n37 = TextureSample(UVs=@n28) {samplerSource=SSM_Wrap_WorldGroupSettings, texture=/VehicleCosmetics/Wheels/_SharedTextures/FX/Trail2_RBG_T}
  n38 = Add(A=Multiply(A=Power(Base=@n35, Exp=@n36), B=Constant() {r=5}), B=Multiply(A=@n37.G, B=Constant() {r=4}))
  n39 = Add(A=AppendVector(A=@n26, B=NamedRerouteUsage()), B=TextureCoordinate())
  n40 = TextureSample(UVs=Rotator(Coordinate=Add(A=@n27, B=@n39)) {speed=0.5}) {samplerSource=SSM_Wrap_WorldGroupSettings, texture=/VehicleCosmetics/Wheels/_SharedTextures/FX/Clouds_T}
  n41 = Clamp(None=@n37.R)
  n42 = Multiply(A=Add(A=Multiply(A=@n38, B=@n40.R), B=@n41), B=@n37.R)
  n43 = Add(A=NamedRerouteUsage(), B=Multiply(A=@n22, B=Add(A=@n34, B=@n42)))
  n44 = NamedRerouteUsage()
  NamedRerouteDeclaration(None=Multiply(A=Add(A=@n32, B=@n33), B=Multiply(A=@n43, B=@n44))) {name=Emissive, nodeColor=<rgba 0.5 0 1 1>}
  n45 = Constant() {r=0.5}
  n46 = Add(A=Multiply(A=@n41, B=Multiply(A=@n37.R, B=@n45)), B=@n37.G)
  n47 = ComponentMask(None=TextureCoordinate()) {r=true}
  n48 = Constant() {r=2}
  n49 = Multiply(
   A=Power(Base=OneMinus(None=@n47), Exp=ScalarParameter() {defaultValue=2.2, parameterName=power fade one minus})
   B=Multiply(A=Constant() {r=50}, B=Power(Base=@n47, Exp=@n48))
  )
  n50 = ScalarParameter() {defaultValue=1, parameterName=TileX}
  n51 = Constant() {r=1}
  n52 = Rotator(Coordinate=Multiply(A=AppendVector(A=@n50, B=@n51), B=TextureCoordinate())) {speed=1.5}
  n53 = Power(
   Base=TextureSample(UVs=@n52) {samplerSource=SSM_Wrap_WorldGroupSettings, texture=/VehicleCosmetics/Wheels/_SharedTextures/FX/Trail1_RBG_T}.G
   Exp=Constant() {r=1.5}
  )
  n54 = Power(
   Base=TextureSample(UVs=@n52) {samplerSource=SSM_Wrap_WorldGroupSettings, texture=&Trail1_RBG_T}.R
   Exp=Constant() {r=2}
  )
  n55 = TextureCoordinate()
  n56 = Multiply(A=ComponentMask(None=@n25) {g=true, r=true}, B=Constant() {r=0.2})
  n57 = Multiply(
   A=Multiply(A=Multiply(A=@n46, B=@n49), B=Multiply(A=@n53, B=@n54))
   B=TextureSample(UVs=Add(A=@n55, B=@n56)) {samplerSource=SSM_Wrap_WorldGroupSettings, texture=/VehicleCosmetics/Wheels/_SharedTextures/FX/TrailMask_M_T}.R
  )
  n58 = NamedRerouteUsage()
  NamedRerouteDeclaration(None=Clamp(None=Multiply(A=@n57, B=@n58))) {name=Opacity, nodeColor=<rgba 0.1016 1 0 1>}
 ]
}
