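# Translucent unlit particle material for the Genie wheel cosmetic's dot FX, usable on sprites, mesh particles and beam trails.
# Particle color drives emissive, and particle alpha times a centered UV sphere mask gives soft round dots without a texture.
# The attributes then pass through the base vehicle damage and vehicle de-rez transition functions, so both vehicle effects can apply to the particles.
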Material /VehicleCosmetics/Wheels/Genie/Materials/M_Wheel_Genie_FX_Dots
graph {
 properties {
  bAutomaticallySetUsageInEditor=false
  bUseMaterialAttributes=true
  bUsedWithBeamTrails=true
  bUsedWithMeshParticles=true
  bUsedWithNiagaraMeshParticles=true
  bUsedWithNiagaraSprites=true
  bUsedWithParticleSprites=true
  blendMode=BLEND_Translucent
  shadingModel=MSM_Unlit
  twoSided=true
  usedShadingModels=Unlit
 }
 nodes [
  n1 = ParticleColor()
  n2 = Multiply(A=@n1.A, B=SphereMask(A=TextureCoordinate(), B=Constant2Vector() {g=0.5, r=0.5}, Radius=Constant() {r=0.5}) {hardnessPercent=0})
  MaterialFunctionCall(In=MaterialFunctionCall(Material=SetMaterialAttributes!(MaterialAttributes, Emissive Color=@n1.RGB, Opacity=@n2) {attributeSetTypes [#1, #2]}) {materialFunction=&MF_BaseVehicle_Damage}.Result) {materialFunction=&MF_VehicleDeRez_Transition}
  VertexColor()
 ]
}
